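; Blueprint interface for the prototype Object Mover weapon, named for overriding creative tool HUD legend text.
; It declares getters for copy, cut and select-hold text, plus a check for whether select-hold text exists.
; The functions here return only empty strings and false, so any actual legend wording comes from implementers.
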(fn GetCopyText
  (return ""))

(fn GetCutText
  (return ""))

(fn GetSelectHoldText
  (return ""))

(fn HasSelectHoldText
  (return false))
