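; Gameplay ability for the Object Mover prototype tool that plays an animation montage with an accompanying sound.
; On activation it commits, plays a configurable sound at the avatar's location, then starts the configured montage through PlayMontageAndWait.
; It ends immediately after starting the montage, whose task is set not to stop it on ability end.
; All montage callbacks and the end-ability event are empty.
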(event Ability|EventActivateAbility
  (Ability|CommitAbility)
  (Audio|PlaySoundatLocation (Variables|Default|GetSoundToPlay) (Transformation|GetActorLocation (Ability|GetAvatarActorfromActorInfo)))
  (Ability|Tasks|PlayMontageAndWait "None" (Variables|Default|GetAnimationMontage) 1.0 "None" false
    (:OnCompleted
      _)
    (:OnBlendOut
      _)
    (:OnInterrupted
      _)
    (:OnCancelled
      _))
  (Ability|EndAbility))

(event Ability|EventOnEndAbility
  (params (bWasCancelled :type Boolean)))
